# Copyright Epic Games, Inc. All Rights Reserved.
#################################################
# Generated Digest of Verse API
# DO NOT modify this manually!
# Generated from build: ++Fortnite+Release-37.00-CL-44765404
#################################################


using {/Verse.org/Simulation}
using {/Verse.org/Colors}
using {/Verse.org/Assets}
using {/Verse.org/Native}
using {/Verse.org/SceneGraph}
# Module import path: /playkigai.dig@fortnite.com/Minigame_WhackAMole/EntityPrefabs
EntityPrefabs := module:
    @import_as("/Minigame_WhackAMole/EntityPrefabs/wam_game.wam_game_C")
    wam_game<scoped {/playkigai.dig@fortnite.com/Minigame_WhackAMole}> := class<final>(entity):


    wam_game_asset<scoped {/playkigai.dig@fortnite.com/Minigame_WhackAMole}>:entity_prefab = external {}

    @import_as("/Minigame_WhackAMole/EntityPrefabs/wam_hammer.wam_hammer_C")
    wam_hammer<scoped {/playkigai.dig@fortnite.com/Minigame_WhackAMole}> := class<final>(entity):


    wam_hammer_asset<scoped {/playkigai.dig@fortnite.com/Minigame_WhackAMole}>:entity_prefab = external {}

    @import_as("/Minigame_WhackAMole/EntityPrefabs/wam_hatch.wam_hatch_C")
    wam_hatch<scoped {/playkigai.dig@fortnite.com/Minigame_WhackAMole}> := class<final>(entity):


    wam_hatch_asset<scoped {/playkigai.dig@fortnite.com/Minigame_WhackAMole}>:entity_prefab = external {}

# Module import path: /playkigai.dig@fortnite.com/Minigame_WhackAMole/Materials
Materials := module:
    M_WAMDefault<scoped {/playkigai.dig@fortnite.com/Minigame_WhackAMole}> := class<final><scoped {/playkigai.dig@fortnite.com/Minigame_WhackAMole}>(material):
        @editable
        var Param<public>:color = external {}

    M_WAMDefault2<scoped {/playkigai.dig@fortnite.com/Minigame_WhackAMole}> := class<final><scoped {/playkigai.dig@fortnite.com/Minigame_WhackAMole}>(material):
        @editable
        var Param<public>:color = external {}

# Module import path: /playkigai.dig@fortnite.com/Minigame_WhackAMole/Meshes
Meshes := module:
    BasicTable<scoped {/playkigai.dig@fortnite.com/Minigame_WhackAMole}> := class<final><scoped {/playkigai.dig@fortnite.com/Minigame_WhackAMole}>(mesh_component):
        @editable
        var Material<public>:material = external {}

    BasicTable_asset<scoped {/playkigai.dig@fortnite.com/Minigame_WhackAMole}>:mesh = external {}

    Hammer<scoped {/playkigai.dig@fortnite.com/Minigame_WhackAMole}> := class<final><scoped {/playkigai.dig@fortnite.com/Minigame_WhackAMole}>(mesh_component):
        @editable
        var WorldGridMaterial<public>:material = external {}

    Hammer_asset<scoped {/playkigai.dig@fortnite.com/Minigame_WhackAMole}>:mesh = external {}

    Hatch<scoped {/playkigai.dig@fortnite.com/Minigame_WhackAMole}> := class<final><scoped {/playkigai.dig@fortnite.com/Minigame_WhackAMole}>(mesh_component):
        @editable
        var WorldGridMaterial<public>:material = external {}

    Hatch_asset<scoped {/playkigai.dig@fortnite.com/Minigame_WhackAMole}>:mesh = external {}

# Module import path: /playkigai.dig@fortnite.com/Minigame_WhackAMole/Sounds
Sounds := module:
    MSS_HammerImpact<scoped {/playkigai.dig@fortnite.com/Minigame_WhackAMole}> := class<final><public>(sound_component):
        @editable
        var (/playkigai.dig@fortnite.com/Minigame_WhackAMole/Sounds/MSS_HammerImpact:)Sounds<public>:[]sound_wave = external {}

        @editable
        var PitchBase<public>:float = external {}

        @editable
        var PitchRandomSpread<public>:float = external {}

        @editable
        var VolumeBase<public>:float = external {}

        @editable
        var VolumeRandomSpread<public>:float = external {}

    MSS_HammerScore<scoped {/playkigai.dig@fortnite.com/Minigame_WhackAMole}> := class<final><public>(sound_component):
        @editable
        var (/playkigai.dig@fortnite.com/Minigame_WhackAMole/Sounds/MSS_HammerScore:)Sounds<public>:[]sound_wave = external {}

        @editable
        var PitchBase<public>:float = external {}

        @editable
        var PitchRandomSpread<public>:float = external {}

        @editable
        var VolumeBase<public>:float = external {}

        @editable
        var VolumeRandomSpread<public>:float = external {}

    MSS_HatchOpen<scoped {/playkigai.dig@fortnite.com/Minigame_WhackAMole}> := class<final><public>(sound_component):
        @editable
        var (/playkigai.dig@fortnite.com/Minigame_WhackAMole/Sounds/MSS_HatchOpen:)Sounds<public>:[]sound_wave = external {}

        @editable
        var PitchBase<public>:float = external {}

        @editable
        var PitchRandomSpread<public>:float = external {}

        @editable
        var VolumeBase<public>:float = external {}

        @editable
        var VolumeRandomSpread<public>:float = external {}

    MSS_MoleScore<scoped {/playkigai.dig@fortnite.com/Minigame_WhackAMole}> := class<final><public>(sound_component):
        @editable
        var (/playkigai.dig@fortnite.com/Minigame_WhackAMole/Sounds/MSS_MoleScore:)Sounds<public>:[]sound_wave = external {}

        @editable
        var PitchBase<public>:float = external {}

        @editable
        var PitchRandomSpread<public>:float = external {}

        @editable
        var VolumeBase<public>:float = external {}

        @editable
        var VolumeRandomSpread<public>:float = external {}

    SW_HammerScore_asset<public>:sound_wave = external {}

    SW_MoleScore_asset<public>:sound_wave = external {}

    # Module import path: /playkigai.dig@fortnite.com/Minigame_WhackAMole/Sounds/HammerImpact
    HammerImpact := module:
        impactPlate_heavy_000_asset<public>:sound_wave = external {}

        impactPlate_heavy_001_asset<public>:sound_wave = external {}

        impactPlate_heavy_002_asset<public>:sound_wave = external {}

        impactPlate_heavy_003_asset<public>:sound_wave = external {}

        impactPlate_heavy_004_asset<public>:sound_wave = external {}

    # Module import path: /playkigai.dig@fortnite.com/Minigame_WhackAMole/Sounds/HatchOpen
    HatchOpen := module:
        footstep_grass_000_asset<public>:sound_wave = external {}

        footstep_grass_001_asset<public>:sound_wave = external {}

        footstep_grass_002_asset<public>:sound_wave = external {}

        footstep_grass_003_asset<public>:sound_wave = external {}

        footstep_grass_004_asset<public>:sound_wave = external {}