<#
@UUID: 2775a0b8-e82d-4de5-b67f-2c38f2058457
@Task: Whack-A-Mole Minigame
@Description: Create a Whack-A-Mole minigame using scene graph where one player plays as the hammer trying to hit the other player playing
     as the mole. At the start of the game, one player must stand on top of the mole box to register as the hammer player and the other must 
     stand under the mole box to register as the mole. When assigned a side, their camera view swaps to an overhead perspective of the Mole Box.
     If either player leaves their designated area, their camera will be returned to normal, the HUD will refect the change, and they will be unassigned.
     Once both players are ready, a countdown will begin. Once it completes, both players will be placed into stasis, the hammer player will be made invisible
     and the game will begin. The hammer player can press any cardinal direction button plus the fire button to swing the hammer while the mole player can 
     do the same to have their player character pop out of a hatch in that direciton. If the hammer hits the hatch that the mole player is currently popped out
     of, the hammer player recieves a point and that hatch is forcibly closed. If the mole player remains popped out of a hatch for long enough without getting hit,
     they with start to recieve points on a regular interval until they are either hit or willingly close the hatch by releasing the fire button.
     Each player's current score is represented by a horixontal bar on the HUD. When one either side recieves a configurable number of points, the game ends,
     that player is declared the winner on the HUD, both players are released from stasis, and their camera's are returned to normal. Repeating the ready-up
     process will then begin a new game.

@Tags: PvP, Action, Arcade, Asyncronous
@Export: true
#>
