<#
@UUID: 2775a0b8-e82d-4de5-b67f-2c38f2058457
@Task: Whack-A-Mole Minigame Utils
@Description: All the miscellaneous helper functions used for this project
@Tags: Utility, DataConversion, Formatting
@Export: true
#>

using { /Verse.org/SceneGraph }
using { /Fortnite.com/Devices }
using { /Verse.org/Simulation }
using { /UnrealEngine.com/Temporary/Diagnostics }

# Removes an element from the given map and returns a new map without that element
(Map:[keytype]valuetype where keytype:subtype(comparable), valuetype:type).RemoveKey(ElementToRemove:keytype)<transacts>:[keytype]valuetype=
    var NewMap:[keytype]valuetype = map{}
    for (Key -> Value : Map, Key <> ElementToRemove):
        set NewMap = ConcatenateMaps(NewMap, map{Key => Value})
    return NewMap

# Helper function to return an array (instead of a generator) of all components attached to descendant entities to and including `InEntity` of type `component_type`.
(Entity:entity).FindDescendantComponentsArray<public>(ComponentType:castable_subtype(component))<transacts>:[]ComponentType =
    for (Comp : Entity.FindDescendantComponents(ComponentType)) do Comp


