<#
@UUID: 2775a0b8-e82d-4de5-b67f-2c38f2058457
@Task: Whack-A-Mole Game Manager Component
@Description: Component that handles the top-level game logic of the whack-a-mole minigame. Handles the player ready system, scoring and fascilitates communication between components
@Tags: SceneGraph, TimerManagement, Gameplay, DataStructures
@Export: true
#>

<#
In-Editor Entity Prefab Setup:
    (NOTE: any setting not mentioned is assumed to be default or to your preference)


wam_hatch
    RootEntity:
        - mesh_component (custom "Hatch" static mesh)

wam_hammer
    RootEntity:
        - wam_hammer_controller_component
        - wam_player_control_component
            MaybeInputUp: input_trigger_device in the level with the following device settigns
                - CreativeInput: Custom 7 (Forward/Back)
                - AxisDirection: Positive
                - RegisterPlayerBehavior: Require Registered
            MaybeInputDown: input_trigger_device in the level with the following device settigns
                - CreativeInput: Custom 7 (Forward/Back)
                - AxisDirection: Negative
                - RegisterPlayerBehavior: Require Registered
            MaybeInputLeft: input_trigger_device in the level with the following device settigns
                - CreativeInput: Custom 7 (Left/Right)
                - AxisDirection: Negative
                - RegisterPlayerBehavior: Require Registered
            MaybeInputRight: input_trigger_device in the level with the following device settigns
                - CreativeInput: Custom 7 (Left/Right)
                - AxisDirection: Positive
                - RegisterPlayerBehavior: Require Registered
            MaybeInputFire: input_trigger_device in the level with the following device settigns
                - CreativeInput: Custom 1 (Fire)
                - ConsumeInput: True
                - RegisterPlayerBehavior: Require Registered
        HammerMesh
            - mesh_component (custom "Hammer" static mesh)
        HammerInpactSound:
            - sound_component (MSS_HammerImpact)
                Autoplay: False
        HammerScoreSound:
            - sound_component (MSS_HammerScore)
                Autoplay: False


wam_game
    RootEntity:
        - wam_hud_component
        - wam_game_manager_component
            MaybeHammerPlayerVolume: a volume_device in the level that covers the area on top of the Mole box Platform near the hammer
            MaybeMolePlayerVolume: a volume_device in the level that covers the area underneath the mole box.
            MaybeTopDownCam: a gameplay_camera_fixed_point_device angled to get a top-down view of the mole box platform with the following device settings
                - Remove On Elimination: True
                - Add To Player On Start: False
    Hammer (instance of wam_hammer entity prefab)
    MoleBoxPlatform:
            - wam_male_box_controller_component
                HatchArray: populate with all wam_hatch prefab instances childed to this entity
            - mesh_component (custom "BasicTable" static mesh)
            - wam_player_control_component
                MaybeInputUp: input_trigger_device in the level with the following device settigns
                    - CreativeInput: Custom 7 (Forward/Back)
                    - AxisDirection: Positive
                    - RegisterPlayerBehavior: Require Registered
                MaybeInputDown: input_trigger_device in the level with the following device settigns
                    - CreativeInput: Custom 7 (Forward/Back)
                    - AxisDirection: Negative
                    - RegisterPlayerBehavior: Require Registered
                MaybeInputLeft: input_trigger_device in the level with the following device settigns
                    - CreativeInput: Custom 7 (Left/Right)
                    - AxisDirection: Negative
                    - RegisterPlayerBehavior: Require Registered
                MaybeInputRight: input_trigger_device in the level with the following device settigns
                    - CreativeInput: Custom 7 (Left/Right)
                    - AxisDirection: Positive
                    - RegisterPlayerBehavior: Require Registered
                MaybeInputFire: input_trigger_device in the level with the following device settigns
                    - CreativeInput: Custom 1 (Fire)
                    - ConsumeInput: True
                    - RegisterPlayerBehavior: Require Registered
        HatchOpenSound
            - sound_component (MSS_HatchOpen)
                Autoplay: False
        MoleScoreSound
            - sound_component (MSS_MoleScore)  
                Autoplay: False
        4x wam_hatch entity prefab instances each covering a hole in the platform
#>

using { /Verse.org }
using { /Verse.org/Concurrency }
using { /Verse.org/Simulation }
using { /Verse.org/Native }
using { /Verse.org/SceneGraph }
using { /Fortnite.com/Game }
using { /Fortnite.com/Characters }
using { /Fortnite.com/Playspaces }
using { /Fortnite.com/Devices }
using { /UnrealEngine.com/Temporary/SpatialMath }
using { /UnrealEngine.com/Temporary/Diagnostics }
using { Sounds }

wam_side := enum {HammerPlayer,MolePlayer}

(Side: wam_side).Opponent()<computes>:wam_side =
    case (Side):
        wam_side.HammerPlayer => wam_side.HammerPlayer
        wam_side.MolePlayer => wam_side.MolePlayer


ToString(Side: wam_side)<transacts>:string =
    case (Side):
        wam_side.HammerPlayer => "Hammer Player"
        wam_side.MolePlayer =>   "Mole Player"


side_info := class():
    Side<public>: wam_side 
    var MaybeAgent<public>: ?agent = false
    var MaybePlayerControlComp<public>: ?wam_player_control_component = false
    var Score<public>: int = 0
    

(SideInfo: side_info).Reset():void =
    set SideInfo.MaybeAgent = false
    set SideInfo.Score = 0


score_update_result := struct:
    ScoringSide: side_info
    NewScore: int


ready_update_result := struct:
    IsHammerSideReady<public>: logic
    IsMoleSideReady<public>: logic


wam_game_manager_component<public> := class<final_super>(component):

    # The number of WHOLE seconds that the pregame countdown will start from once both teams are ready
    @editable_number(int):
        MinValue := option. 0
    CountdownStartTime: int = 3

    # The amount of time in seconds that must pass after the game is over before it resets ready for the next game
    @editable_number(float):
        MinValue := option. 0.0
    PostGameDelay: float = 8.0

    # This value multiplied by the number of players on the largest team equals the winning score
    @editable
    ScoreGoal: int = 15

    # Number of points that subtraced when a player hits a teammate (set to 0 to disable friendlyfire)
    @editable_number(int):
        MinValue := option. 0
    FriendlyFirePenalty: int = 0

    # How quickly moles score points
    @editable_number(float):
        MinValue := option. 0.0
    MoleScoreInterval: float = 0.60
    
    @editable
    var MaybeHammerPlayerVolume: ?volume_device = false

    @editable
    var MaybeMolePlayerVolume: ?volume_device = false

    @editable
    var MaybeTopDownCam: ?gameplay_camera_fixed_point_device = false

    # ============ EVENTS =======================================

    # Signals each time when a player is added or removed from a one of the two teams during the pre game; passes the number of members on each team
    ReadyStatusUpdateEvent: event(ready_update_result) = event(ready_update_result){}

    # Signals each time the pregame countdown ticks down; passes the new timer value
    CountdownTimerUpdatedEvent<public>: event(int) = event(int){}

    # Signals after the pregame countdown timer expires and when the game starts
    GameBeginEvent<public>: event() = event(){}

    # Signals when either team's score is updated; passes the team whose score was updated as well as their new score value.
    ScoreUpdatedEvent: event(score_update_result) = event(score_update_result){}

    # Signals when the game ends (regardless of the result); passes the winning team's info struct (passes false if there was no winner)
    GameOverEvent: event(?side_info) = event(?side_info){}

    # Signals when the Reset function is called and when the Post game delay is over.
    GameResetEvent<public>: event() = event(){}

    # Signals when the mole is hit
    MoleHit<public>: event() = event(){}

    # Signals when this component's OnEndSimulation function fires; used to cancel asyncronous logic when this component stops simulating
    EndSimulationEvent<internal>: event() = event(){}

    # ============ INTERNAL FIELDS =======================================

    var HammerSide: side_info = side_info{Side := wam_side.HammerPlayer}
    var MoleSide: side_info = side_info{Side := wam_side.MolePlayer}
    var MaybeHUDComponent: ?wam_hud_component = false  # Reference to the sibling wam_hud_component
    var LastMoleHatchUpdateResult : hatch_status_update = hatch_status_update{}


    OnSimulate<override>()<suspends>:void=
        set MaybeHUDComponent = option. Entity.GetComponent[wam_hud_component] 
        loop:
            if (SimEntity := Entity.GetSimulationEntity[]):

                if (RoundMan := SimEntity.GetFortRoundManager[]):
                    RoundMan.SubscribeRoundStarted(OnRoundStart)
                    break

            Print("Didn't find round manager")
            Sleep(0.0)
        

    OnEndSimulation<override>():void=
        (super:)OnEndSimulation()
        EndSimulationEvent.Signal()


    OnRoundStart()<suspends>: void =
        set HammerSide.MaybePlayerControlComp = option. Entity.FindDescendantComponentsArray(wam_hammer_controller_component)[0].MaybePlayerControlComp?
        set MoleSide.MaybePlayerControlComp = option. Entity.FindDescendantComponentsArray(wam_mole_box_controller_component)[0].MaybePlayerControlComp?
        
        if(MoleBoxComp := Entity.FindDescendantComponentsArray(wam_mole_box_controller_component)[0]):
            set MoleBoxComp.MaybeGameManager = option. Self

        race:
            EndSimulationEvent.Await()
            loop:
                ResetGame()
                AwaitBothSidesReady()


                GameBeginEvent.Signal()
                Print("wam_game_manager_component\nGame Start! First team to score {ScoreGoal} is the winner!")

                spawn. ListenForScoreUpdateHammer()
                spawn. ListenForScoreUpdateMole()

                PutPlayersInGameMode()

                # On Game Start
                MaybeWinner :?side_info = race: 
                    ListenForEitherSideAbandon()
                    ListenForWinner()

                # Post Game
                if (Winner := MaybeWinner?):
                    Print("wam_game_manager_component\n{ToString(Winner.Side)} is the winner!")
                else:
                    Print("wam_game_manager_component\nNoone won. It is a draw...somehow.")
                ReleasePlayersFromGameMode()
                Sleep(0.0) # Wait one frame so the last point counts in the HUD component
                GameOverEvent.Signal(MaybeWinner)   
                Sleep(PostGameDelay)
    

    AwaitBothSidesReady()<suspends>:void =
        if:
            HammerPlayerVolume := MaybeHammerPlayerVolume?
            MolePlayerVolume := MaybeMolePlayerVolume?
        then:
            race:
                GameBeginEvent.Await()
                EndSimulationEvent.Await()
                ListenForSideReadyStatusUpdate(wam_side.HammerPlayer, HammerPlayerVolume)
                ListenForSideReadyStatusUpdate(wam_side.MolePlayer, MolePlayerVolume)
                loop:   # Until they're both ready and the game starts
                    if (BothPlayersReady[]):
                        WasCountdownCompleted := InitiateCountdown(CountdownStartTime)
                        if (WasCountdownCompleted?) then break
    
                    ReadyStatusUpdateEvent.Await()
                    Sleep(0.0)


    PutPlayersInGameMode():void=
        if:
            HammerAgent := HammerSide.MaybeAgent?
            HammerFortChar := HammerAgent.GetFortCharacter[]
        then:
            HammerFortChar.PutInStasis(stasis_args {AllowTurning := false, AllowFalling := false, AllowEmotes := false})
            HammerFortChar.Hide()

        if:
            MoleAgent := MoleSide.MaybeAgent?
            MoleFortChar := MoleAgent.GetFortCharacter[]
        then:
            MoleFortChar.PutInStasis(stasis_args {AllowTurning := false, AllowFalling := false, AllowEmotes := false})


    ReleasePlayersFromGameMode():void=
        if:
            HammerAgent := HammerSide.MaybeAgent?
            HammerFortChar := HammerAgent.GetFortCharacter[]
        then:
            HammerFortChar.ReleaseFromStasis()
            HammerFortChar.Show()

        if:
            MoleAgent := MoleSide.MaybeAgent?
            MoleFortChar := MoleAgent.GetFortCharacter[]
        then:
            MoleFortChar.ReleaseFromStasis()


    # Counts down one second at a time starting from the starting count 
    # Returns true if count reaches 0, returns false if either team is no longer "ready" or end simulation event signals before the count expires
    InitiateCountdown(StartingCount: int)<suspends>:logic =
        var  CountdownTimer: int = StartingCount   
        race:
            EndSimulationEvent.Await()
            loop:
                if (not BothPlayersReady[]) then break
                Sleep(0.0)
                ReadyStatusUpdateEvent.Await()
            loop:
                CountdownTimerUpdatedEvent.Signal(CountdownTimer)
                if (CountdownTimer <= 0) then return true 
                Sleep(1.0)
                set CountdownTimer -= 1
        false
            

    # Succeeds if both player slots are ready
    BothPlayersReady()<decides><transacts>:void = HammerSide.MaybeAgent? and MoleSide.MaybeAgent?
    

    ListenForSideReadyStatusUpdate(Side: wam_side, Volume: volume_device)<suspends>: void =
        SideInfo := GetSideInfo(Side)
        race:
            EndSimulationEvent.Await()
            loop:
                # determine if we should be listening for an unregister or a register
                if (RegisteredAgent := SideInfo.MaybeAgent?):
                    LeavingAgent := Volume.AgentExitsEvent.Await()

                    if (LeavingAgent = RegisteredAgent):
                        UnassignAgentFromSide(Side)
                else:
                    EnteringAgent := Volume.AgentEntersEvent.Await()
                    AssignAgentToSide(EnteringAgent, Side)

                Sleep(0.0)


    # Continuously Listens for when the agent assigned to either side gets removed from the playspace
    ListenForEitherSideAbandon()<suspends>:?side_info =
        if (Playspace := Entity.GetPlayspaceForEntity[]):
            race:
                EndSimulationEvent.Await()
                loop: 
                    QuittingAgent := Playspace.ParticipantRemovedEvent().Await()
                    if (QuittingSide := WhichSide(QuittingAgent)?):
                        UnassignAgentFromSide(QuittingSide.Side)
                        return option. QuittingSide
                    Sleep(0.0)
        false


    ListenForWinner()<suspends>:?side_info =
        loop:  
            Result := ScoreUpdatedEvent.Await()
            Print("{Result.NewScore}, {ScoreGoal}")
            if (Result.NewScore >= ScoreGoal):
                return option. Result.ScoringSide
            Sleep(0.0)


    ListenForScoreUpdateHammer()<suspends>:void =
        MaybeHammerControlComp := option. Entity.FindDescendantComponentsArray(wam_hammer_controller_component)[0]
        if(HammerControlComp := MaybeHammerControlComp?):
            race:
                EndSimulationEvent.Await()
                GameOverEvent.Await()
                loop:  
                    Result := HammerControlComp.HammerStatusUpdated.Await()
                    if:
                        LastMoleHatchUpdateResult.IsOpen = true
                        LastMoleHatchUpdateResult.HatchDir = Result
                    then:
                        # Play a hammer score sound
                        SoundComp := Entity.FindDescendantComponents(MSS_HammerScore)
                        for(ThisSoundComp : SoundComp):
                            ThisSoundComp.Play()

                        set HammerSide.Score += 1
                        MoleHit.Signal()
                        ScoreUpdatedEvent.Signal(score_update_result{ScoringSide := HammerSide, NewScore := HammerSide.Score})

                    Sleep(0.0)
        else:
            Print("Hammer control component cannot be found")


    ListenForScoreUpdateMole()<suspends>:void =
        MaybeMoleControlComp := option. Entity.FindDescendantComponentsArray(wam_mole_box_controller_component)[0]
        if(MoleControlComp := MaybeMoleControlComp?):
            race:
                EndSimulationEvent.Await()
                GameOverEvent.Await()
                loop:  
                    Result := MoleControlComp.HatchStatusUpdated.Await()
                    set LastMoleHatchUpdateResult = Result

                    if(LastMoleHatchUpdateResult.IsOpen = true):
                        spawn. GiveMolePointsLoop(MoleControlComp)

                    Sleep(0.0)
        else:
            Print("Mole control component cannot be found")


    GiveMolePointsLoop(MoleControlComp : wam_mole_box_controller_component)<suspends>:void=
        race:
            EndSimulationEvent.Await()
            GameOverEvent.Await()
            loop:
                Result := MoleControlComp.HatchStatusUpdated.Await()
                if(Result.IsOpen = false):
                    break
                Sleep(0.0)
            loop:
                Sleep(MoleScoreInterval)
                # Play a mole score sound
                SoundComp := Entity.FindDescendantComponents(MSS_MoleScore)
                for(ThisSoundComp : SoundComp):
                    ThisSoundComp.Play()
                set MoleSide.Score += 1
                ScoreUpdatedEvent.Signal(score_update_result{ScoringSide := MoleSide, NewScore := MoleSide.Score})


    UpdateReadyStatuses():void =
        IsHammerSideReady := logic. HammerSide.MaybeAgent?
        IsMoleSideReady := logic. MoleSide.MaybeAgent?
        ReadyStatusUpdateEvent.Signal(ready_update_result{ 
            IsHammerSideReady := IsHammerSideReady,
            IsMoleSideReady := IsMoleSideReady
        })
    

    AssignAgentToSide(Agent: agent,Side: wam_side):void =
        SideInfo := GetSideInfo(Side)
        if:
            not SideInfo.MaybeAgent?
            PlayerControlComp := SideInfo.MaybePlayerControlComp?
            TopDownCam := MaybeTopDownCam?
            # FortCharacter := Agent.GetFortCharacter[]
        then:
            Print("Assigning Agent to {Side}")
            set SideInfo.MaybeAgent = option. Agent
            TopDownCam.Enable()
            TopDownCam.AddTo(Agent)
            # FortCharacter.PutInStasis(stasis_args {AllowTurning := false, AllowFalling := false, AllowEmotes := false})
            PlayerControlComp.BindAgent(Agent)
            UpdateReadyStatuses()


    UnassignAgentFromSide(Side: wam_side):void =
        SideInfo := GetSideInfo(Side)
        if:
            AssignedAgent := SideInfo.MaybeAgent?
            PlayerControlComp := SideInfo.MaybePlayerControlComp?
            TopDownCam := MaybeTopDownCam?
            FortCharacter := AssignedAgent.GetFortCharacter[]
        then:
            Print("Unassign Agent from {Side}")
            TopDownCam.RemoveFrom(AssignedAgent)
            FortCharacter.ReleaseFromStasis()
            PlayerControlComp.UnbindAgent()
            set SideInfo.MaybeAgent = false
            UpdateReadyStatuses()
    
    
    
    GetSideInfo(Side: wam_side)<transacts>:side_info =
        case (Side):
            wam_side.HammerPlayer => HammerSide
            wam_side.MolePlayer => MoleSide

    
    # Succeeds and returns a side_info data structure if the agent is registered one of the two sides; fails otherwise
    WhichSide(Agent : agent)<transacts>:?side_info =
        if (Agent = HammerSide.MaybeAgent?): 
            return option. HammerSide
        else if (Agent = MoleSide.MaybeAgent?):
            return option. MoleSide
        false


    ResetGame(): void = 
        UnassignAgentFromSide(wam_side.HammerPlayer)
        HammerSide.Reset()
        UnassignAgentFromSide(wam_side.MolePlayer)
        MoleSide.Reset()
        GameResetEvent.Signal()
    
    

            

    
            
        
    
