<#
@UUID: 2775a0b8-e82d-4de5-b67f-2c38f2058457
@Task: Whack-A-Mole Hammer Component
@Description: Allows a player bound to a sibling wam_player_control_component to pop in and out of four hatches. Pressing a directional input plus the fire button allows the player to swing the hammer down in that direction. Hitting the mole player while they are popped out will cause the mole to have their hatch closed, the hammer player to get a point, and a sound effect to be played.
@Tags: Gameplay, Movement
@Export: true
#>

<#
In-Editor Entity Prefab Setup:
    (NOTE: any setting not mentioned is assumed to be default or to your preference)

wam_hammer
    RootEntity:
        - wam_hammer_controller_component
        - wam_player_control_component
            MaybeInputUp: input_trigger_device in the level with the following device settigns
                - CreativeInput: Custom 7 (Forward/Back)
                - AxisDirection: Positive
                - RegisterPlayerBehavior: Require Registered
            MaybeInputDown: input_trigger_device in the level with the following device settigns
                - CreativeInput: Custom 7 (Forward/Back)
                - AxisDirection: Negative
                - RegisterPlayerBehavior: Require Registered
            MaybeInputLeft: input_trigger_device in the level with the following device settigns
                - CreativeInput: Custom 7 (Left/Right)
                - AxisDirection: Negative
                - RegisterPlayerBehavior: Require Registered
            MaybeInputRight: input_trigger_device in the level with the following device settigns
                - CreativeInput: Custom 7 (Left/Right)
                - AxisDirection: Positive
                - RegisterPlayerBehavior: Require Registered
            MaybeInputFire: input_trigger_device in the level with the following device settigns
                - CreativeInput: Custom 1 (Fire)
                - ConsumeInput: True
                - RegisterPlayerBehavior: Require Registered
        HammerMesh
            - mesh_component (using custom "Hammer" static mesh)
        HammerInpactSound:
            - MSS_HammerImpact
                Autoplay: False
        HammerScoreSound:
            - MSS_HammerScore
                Autoplay: False
#>


using { /Verse.org }
using { /Verse.org/Concurrency }
using { /Verse.org/Simulation }
using { /Verse.org/Native }
using { /Verse.org/SceneGraph }
using { /Fortnite.com/Game }
using { /Fortnite.com/Teams }
using { /Fortnite.com/Characters }
using { /Fortnite.com/Playspaces }
using { /Fortnite.com/Devices }
using { /UnrealEngine.com/Temporary/SpatialMath }
using { /UnrealEngine.com/Temporary/Diagnostics }
using { /Verse.org/SpatialMath }
using { Sounds }

wam_hammer_controller_component<public> := class<final_super>(component):

    var CurrentDir: wam_dir = wam_dir.Right
    var CanInput: logic = true

    # Signals when this component's OnEndSimulation function fires; used to cancel asyncronous logic when this component stops simulating
    EndSimulationEvent<internal>: event() = event(){}
    # Signals when the input status has updated
    InputStatusUpdated: event() = event(){}
    # Signals when the hammer swings; passes the direction the hammer has been swung in
    HammerStatusUpdated: event(wam_dir) = event(wam_dir){}

    var MaybePlayerControlComp: ?wam_player_control_component = false

    @editable
    HammerSwingUpSpeed : float = 2.0
    @editable
    HammerSwingDownSpeed : float = 9.0

    OnBeginSimulation<override>():void=
        (super:)OnBeginSimulation()
        set MaybePlayerControlComp = option. Entity.GetComponent[wam_player_control_component]
        
        if (RoundMan := Entity.GetFortRoundManager[]):
            RoundMan.SubscribeRoundStarted(OnRoundStart)
        else Print("Didn't find round manager")
        

    OnEndSimulation<override>():void=
        (super:)OnEndSimulation()
        EndSimulationEvent.Signal()


    OnRoundStart()<suspends>: void =
        
        if (PlayerControlComp := MaybePlayerControlComp?):
            race:
                EndSimulationEvent.Await()
                loop:   # On Input Fire Pressed
                    PlayerControlComp.FirePressedEvent.Await()
                    if (CanInput?) then SwingHammer()
                    Sleep(0.0)
                loop:   # On Input Up Pressed
                    PlayerControlComp.UpPressedEvent.Await()
                    set CurrentDir = wam_dir.Up
                    Sleep(0.0)
                loop:   # On Input Down Pressed
                    PlayerControlComp.DownPressedEvent.Await()
                    set CurrentDir = wam_dir.Down
                    Sleep(0.0)
                loop:   # On Input Left Pressed
                    PlayerControlComp.LeftPressedEvent.Await()
                    set CurrentDir = wam_dir.Left
                    Sleep(0.0)
                loop:   # On Input Right Pressed
                    PlayerControlComp.RightPressedEvent.Await()
                    set CurrentDir = wam_dir.Right
                    Sleep(0.0)


    SwingHammer()<suspends>: void =
        var HammerYaw : float = 0.0

        case (CurrentDir):
            wam_dir.Up => set HammerYaw = 90.0
            wam_dir.Down => set HammerYaw = 270.0
            wam_dir.Left => set HammerYaw = 0.0
            wam_dir.Right => set HammerYaw = 180.0

        InputStatusUpdated.Signal()

        StartRotation := (/Verse.org/SpatialMath:) MakeRotationFromYawPitchRollDegrees(HammerYaw, 0.0, 0.0)
        EndRotation := (/Verse.org/SpatialMath:) MakeRotationFromYawPitchRollDegrees(HammerYaw, 0.0, 90.0)
        LerpLocalRotation(HammerSwingDownSpeed, StartRotation, EndRotation)
        HammerStatusUpdated.Signal(CurrentDir)
        # Play a hammer score sound
        SoundComp := Entity.FindDescendantComponents(MSS_HammerImpact)
        for(ThisSoundComp : SoundComp):
            ThisSoundComp.Play()
        LerpLocalRotation(HammerSwingUpSpeed, EndRotation, StartRotation)


    LerpLocalRotation(MoveSpeed : float, StartRotation: (/Verse.org/SpatialMath:) rotation, EndRotation: (/Verse.org/SpatialMath:) rotation)<suspends>:void=
        #set CanInput = false
        race:
            EndSimulationEvent.Await()
            InputStatusUpdated.Await()
            block:
                Timestamp := GetSimulationElapsedTime()
                loop:
                    Sleep(0.0)
                    DeltaTime := GetSimulationElapsedTime() - Timestamp
                    Interp := Clamp(MoveSpeed*DeltaTime, 0.0, 1.0)

                    NewRotation := Slerp(StartRotation, EndRotation, Interp)

                    NewTransform := (/Verse.org/SpatialMath:) transform{
                        Translation:=Entity.GetLocalTransform().Translation, 
                        Rotation:=NewRotation, 
                        Scale:=Entity.GetLocalTransform().Scale
                    }
                    
                    Entity.SetLocalTransform(NewTransform)

                    if(Interp >= 1.0):
                        #set CanInput = true
                        break

            
