<#
@UUID: 2775a0b8-e82d-4de5-b67f-2c38f2058457
@Task: Whack-A-Mole Mole Box Controller Component
@Description: Allows a player bound to a sibling wam_player_control_component to pop in and out of four hatches. Pressing a directional input plus the fire button allows the player to open a hatch and have their character model pop up from the hole, releasing the fire button will cause them to drop back down and close the hatch. If the player sticks out for long enough without getting hit by the hammer they score a point and a sound effect is played
@Tags: Gameplay, Movement, PlayerManagement
@Export: true
#>

using { /Verse.org }
using { /Verse.org/Concurrency }
using { /Verse.org/Simulation }
using { /Verse.org/Native }
using { /Verse.org/SceneGraph }
using { /Fortnite.com/Game }
using { /Fortnite.com/Teams }
using { /Fortnite.com/Characters }
using { /Fortnite.com/Playspaces }
using { /Fortnite.com/Devices }
using { /UnrealEngine.com/Temporary/SpatialMath }
using { /UnrealEngine.com/Temporary/Diagnostics }
using { /Verse.org/SpatialMath }
using {EntityPrefabs}
using {Sounds}

hatch_status_update := struct:
    HatchDir : wam_dir = wam_dir.Down
    IsOpen : logic = false

wam_mole_box_controller_component<public> := class<final_super>(component):

    # Signals when this component's OnEndSimulation function fires; used to cancel asyncronous logic when this component stops simulating
    EndSimulationEvent<internal>: event() = event(){}
    # Signals when the hammer status has updated
    InputStatusUpdated: event() = event(){}
    # Signals when the hatch status updates
    HatchStatusUpdated: event(hatch_status_update) = event(hatch_status_update){}

    var CurrentDir: wam_dir = wam_dir.Right
    var MaybePlayerControlComp: ?wam_player_control_component = false
    var MaybeGameManager : ?wam_game_manager_component = false
    var CanInput: logic = true

    @editable 
    HatchArray : []wam_hatch = array{}
    @editable
    HatchOpenSpeed : float = 6.0
    @editable
    HatchCloseSpeed : float = 3.0
    @editable
    PlayerBelowHatchDistance : float = -150.0
    @editable
    PlayerAboveHatchDistance : float = 0.0

    OnBeginSimulation<override>():void=
        (super:)OnBeginSimulation()
        set MaybePlayerControlComp = option. Entity.GetComponent[wam_player_control_component]
        
        if (RoundMan := Entity.GetFortRoundManager[]):
            RoundMan.SubscribeRoundStarted(OnRoundStart)
        else Print("Didn't find round manager")
        

    OnEndSimulation<override>():void=
        (super:)OnEndSimulation()
        EndSimulationEvent.Signal()


    OnRoundStart()<suspends>: void =
          if (PlayerControlComp := MaybePlayerControlComp?):
            race:
                EndSimulationEvent.Await()
                loop:   # On Input Fire Pressed
                    PlayerControlComp.FirePressedEvent.Await()
                    if (CanInput?) then OnInputPressed()
                    Sleep(0.0)
                loop:   # On Input Up Pressed
                    PlayerControlComp.UpPressedEvent.Await()
                    if (CanInput?) then set CurrentDir = wam_dir.Up
                    Sleep(0.0)
                loop:   # On Input Down Pressed
                    PlayerControlComp.DownPressedEvent.Await()
                    if (CanInput?) then set CurrentDir = wam_dir.Down
                    Sleep(0.0)
                loop:   # On Input Left Pressed
                    PlayerControlComp.LeftPressedEvent.Await()
                    if (CanInput?) then set CurrentDir = wam_dir.Left
                    Sleep(0.0)
                loop:   # On Input Right Pressed
                    PlayerControlComp.RightPressedEvent.Await()
                    if (CanInput?) then set CurrentDir = wam_dir.Right
                    Sleep(0.0)


    OnInputPressed()<suspends>: void =
        if (DefaultHatchEntity := HatchArray[0]):
            var HatchEntity : wam_hatch = DefaultHatchEntity
            case (CurrentDir):
                wam_dir.Up =>
                    if(NewHatchEntity := HatchArray[0]):
                        set HatchEntity = NewHatchEntity
                wam_dir.Down =>
                    if(NewHatchEntity := HatchArray[1]):
                        set HatchEntity = NewHatchEntity
                wam_dir.Left =>
                    if(NewHatchEntity := HatchArray[2]):
                        set HatchEntity = NewHatchEntity
                wam_dir.Right =>
                    if(NewHatchEntity := HatchArray[3]):
                        set HatchEntity = NewHatchEntity

            InputStatusUpdated.Signal()
            CloseAllHatchesExceptFor(HatchEntity)

            # Open hatch
            OpenHatch(HatchEntity)
            MovePlayerUpOrDown(HatchEntity, true)
            HatchStatusUpdated.Signal(hatch_status_update{HatchDir := CurrentDir, IsOpen := true})

            # Close the hatch
            if (PlayerControlComp := MaybePlayerControlComp?):
                AwaitPlayerNotHoldingButton()
                CloseHatch(HatchEntity)
                MovePlayerUpOrDown(HatchEntity, false)
                HatchStatusUpdated.Signal(hatch_status_update{HatchDir := CurrentDir, IsOpen := false})
        else:
            Print("Hatches not found. Did you forget to populate the hatch array on the mole box controller component?")


    AwaitPlayerNotHoldingButton()<suspends>:void=
        if:
            GameManager := MaybeGameManager?
            PlayerControlComp := MaybePlayerControlComp?
            FireInputTriggerDevice := PlayerControlComp.MaybeInputFire?
            BoundPlayer := PlayerControlComp.MaybeBoundAgent?
            not FireInputTriggerDevice.IsHeld[BoundPlayer]
        then:
            race:
                GameManager.MoleHit.Await()
                loop:
                    ReleasedAgent := FireInputTriggerDevice.ReleasedEvent.Await()
                    if(ReleasedAgent(0) = BoundPlayer):
                        break
                    Sleep(0.0)


    GetNewEntityRotationFromRoll(NewHatchEntity : wam_hatch, Roll : float):(/Verse.org/SpatialMath:)rotation=
        YawPitchRoll := NewHatchEntity.GetLocalTransform().Rotation.GetYawPitchRollDegrees()
        Yaw := YawPitchRoll(0)
        Pitch := YawPitchRoll(1)
        return (/Verse.org/SpatialMath:) MakeRotationFromYawPitchRollDegrees(Yaw, Pitch, Roll)


    CloseHatch(HatchEntity : wam_hatch)<suspends>:void=
        StartRotation := HatchEntity.GetLocalTransform().Rotation
        EndRotation := GetNewEntityRotationFromRoll(HatchEntity, 0.0)
        spawn. LerpLocalRotation(HatchEntity, HatchCloseSpeed, StartRotation, EndRotation)


    OpenHatch(HatchEntity : wam_hatch)<suspends>:void=
        StartRotation := HatchEntity.GetLocalTransform().Rotation
        EndRotation := GetNewEntityRotationFromRoll(HatchEntity, 90.0)
        spawn. LerpLocalRotation(HatchEntity, HatchOpenSpeed, StartRotation, EndRotation)

        # Play a hatch open sound
        SoundComp := Entity.FindDescendantComponents(MSS_HatchOpen)
        for(ThisSoundComp : SoundComp):
            ThisSoundComp.Play()

    
    MovePlayerUpOrDown(HatchEntity : wam_hatch, IsUp : logic)<suspends>:void=
        if:
            PlayerControlComp := MaybePlayerControlComp?
            BoundAgent := PlayerControlComp.MaybeBoundAgent?
        then:
            OldPlayerRotation := HatchEntity.GetGlobalTransform().Rotation.GetYawPitchRollDegrees()
            Yaw := OldPlayerRotation(0) - 90.0
            Pitch := OldPlayerRotation(1)
            NewPlayerRotation := (/Verse.org/SpatialMath:) MakeRotationFromYawPitchRollDegrees(Yaw, Pitch, 0.0)
            NewRot := FromRotation(NewPlayerRotation)
            var NormalizedVector3 : (/UnrealEngine.com/Temporary/SpatialMath:) vector3 = NewRot.GetLocalForward()
            set NormalizedVector3.Y = NormalizedVector3.Y*-1 # Y needs to be flipped so offset works properly
            if(GameScale := FromVector3(Entity.GetParent[].GetGlobalTransform().Scale)):
                PlayerOffset := FromVector3(NormalizedVector3*GameScale*100.0)

                StartTranslation := (/Verse.org/SpatialMath:) vector3{
                        Left:=HatchEntity.GetGlobalTransform().Translation.Left, 
                        Forward:=HatchEntity.GetGlobalTransform().Translation.Forward, 
                        Up:=HatchEntity.GetGlobalTransform().Translation.Up + PlayerBelowHatchDistance
                    }
                EndTranslation := (/Verse.org/SpatialMath:) vector3{
                        Left:=HatchEntity.GetGlobalTransform().Translation.Left, 
                        Forward:=HatchEntity.GetGlobalTransform().Translation.Forward, 
                        Up:=HatchEntity.GetGlobalTransform().Translation.Up + PlayerAboveHatchDistance
                    }
                
                if(IsUp?):
                    spawn. LerpAgentGlobalTranslationRotation(BoundAgent, HatchOpenSpeed, StartTranslation+PlayerOffset, EndTranslation+PlayerOffset, NewPlayerRotation)
                else: # If the player needs to be moved down, swap the end and start positions
                    spawn. LerpAgentGlobalTranslationRotation(BoundAgent, HatchCloseSpeed, EndTranslation+PlayerOffset, StartTranslation+PlayerOffset, NewPlayerRotation)
            else:
                Print("Failed to get game scale, can't set player position")


    CloseAllHatchesExceptFor(HatchEntityToExclude : wam_hatch)<suspends>:void=
        for:
            HatchEntity: HatchArray
            HatchEntity <> HatchEntityToExclude
        do:
            CloseHatch(HatchEntity)


    CloseAllHatches()<suspends>:void=
        for(HatchEntity: HatchArray):
            CloseHatch(HatchEntity)
        

    LerpLocalRotation(HatchEntity : wam_hatch, MoveSpeed : float, StartRotation: (/Verse.org/SpatialMath:) rotation, EndRotation: (/Verse.org/SpatialMath:) rotation)<suspends>:void=
        race:
            EndSimulationEvent.Await()
            InputStatusUpdated.Await()
            block:
                Timestamp := GetSimulationElapsedTime()
                loop:
                    Sleep(0.0)
                    DeltaTime := GetSimulationElapsedTime() - Timestamp
                    Interp := Clamp(MoveSpeed*DeltaTime, 0.0, 1.0)
                    NewRotation := Slerp(StartRotation, EndRotation, Interp)

                    NewTransform := (/Verse.org/SpatialMath:) transform{
                        Translation:=HatchEntity.GetLocalTransform().Translation, 
                        Rotation:=NewRotation, 
                        Scale:=HatchEntity.GetLocalTransform().Scale
                    }
                    HatchEntity.SetLocalTransform(NewTransform)
                    if(Interp >= 1.0):
                        break


    LerpAgentGlobalTranslationRotation(Agent : agent, MoveSpeed : float, StartTranslation: (/Verse.org/SpatialMath:) vector3, EndTranslation: (/Verse.org/SpatialMath:) vector3, Rotation : (/Verse.org/SpatialMath:) rotation)<suspends>:void=
        race:
            EndSimulationEvent.Await()
            InputStatusUpdated.Await()
            block:
                Timestamp := GetSimulationElapsedTime()
                loop:
                    Sleep(0.0)
                    DeltaTime := GetSimulationElapsedTime() - Timestamp
                    Interp := Clamp(MoveSpeed*DeltaTime, 0.0, 1.0)
                    NewTranslation := Lerp(StartTranslation, EndTranslation, Interp)
                    if:
                        AgentFortCharacter := Agent.GetFortCharacter[]
                        NewRotation := FromRotation(Rotation)
                        NewVector3 := FromVector3(NewTranslation)
                        AgentFortCharacter.TeleportTo[NewVector3, NewRotation]

                    if(Interp >= 1.0):
                        break