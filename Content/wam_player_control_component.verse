<#
@UUID: 28c43ada-f9b8-40e2-86be-01773dff5198
@Task: Whack-A-Mole Player Control Component
@Description: Component that uses input_trigger_devices to detect when a specific player agent has inputted either up, down, left, right or the fire input. 
@Tags: PlayerInteraction, PlayerManagement, CreativeDevices
@Export: true
#>


using { /Fortnite.com/Devices }
using { /Verse.org/SceneGraph }
using { /Verse.org/Simulation }
using { /Fortnite.com/Game }
using { /Fortnite.com/Devices }
using { /Fortnite.com/Characters }
using { /Fortnite.com/Playspaces }
using { /Verse.org/Assets }

wam_dir := enum {Up,Down,Left,Right}

wam_player_control_component<public> := class<final_super>(component):

    @editable
    var MaybeInputUp: ?input_trigger_device = false
    @editable
    var MaybeInputDown: ?input_trigger_device = false
    @editable
    var MaybeInputRight: ?input_trigger_device = false
    @editable
    var MaybeInputLeft: ?input_trigger_device = false
    @editable
    var MaybeInputFire: ?input_trigger_device = false

    # Signals when an agent becomes bound to this component; passes the newly bound agent
    AgentBoundEvent<public>: event(agent) = event(agent){}
    # Signals when an agent becomes unbound to this component; passes the previously bound agent
    AgentUnboundEvent<public>: event(agent) = event(agent){}

    # Signals when the registered player presses the UP input
    UpPressedEvent<public>: event() = event(){}
    # Signals when the registered player presses the DOWN input
    DownPressedEvent<public>: event() = event(){}
    # Signals when the registered player presses the LEFT input
    LeftPressedEvent<public>: event() = event(){}
    # Signals when the registered player presses the RIGHT input
    RightPressedEvent<public>: event() = event(){}
    # Signals when the registered player presses the FIRE input
    FirePressedEvent<public>: event() = event(){}
    # Signals when the input devices have been reinitialized 
    InputsInitializedEvent<internal>: event() = event(){}
    # Signals when this component's OnEndSimulation function fires; used to cancel asyncronous logic when this component stops simulating
    EndSimulationEvent<public>: event() = event(){}

    var AllInputDevices:[]input_trigger_device = array{}
    var MaybeBoundAgent: ?agent = false


    OnBeginSimulation<override>():void=
        (super:)OnBeginSimulation()
        if (RoundMan := Entity.GetFortRoundManager[]):
            RoundMan.SubscribeRoundStarted(OnRoundStart)
        if (Playspace := Entity.GetPlayspaceForEntity[]):
            Playspace.ParticipantRemovedEvent().Subscribe(OnAgentRemoved)
    

    OnRoundStart()<suspends>: void =
        InitializeInputDevices()


    InitializeInputDevices():void =
        InputsInitializedEvent.Signal()
        if (BoundAgent := MaybeBoundAgent?) then UnregisterFromAllDevices(BoundAgent)
        set AllInputDevices = array{}
        if (InputUp := MaybeInputUp?):
            InputUp.Enable()
            spawn. InputTriggerListenerPressed(InputUp, UpPressedEvent)
            set AllInputDevices += array{InputUp}
        if (InputDown := MaybeInputDown?):
            InputDown.Enable()
            spawn. InputTriggerListenerPressed(InputDown, DownPressedEvent)
            set AllInputDevices += array{InputDown}
        if (InputRight := MaybeInputRight?):
            InputRight.Enable()
            spawn. InputTriggerListenerPressed(InputRight, RightPressedEvent)
            set AllInputDevices += array{InputRight}
        if (InputLeft := MaybeInputLeft?):
            InputLeft.Enable()
            spawn. InputTriggerListenerPressed(InputLeft, LeftPressedEvent)
            set AllInputDevices += array{InputLeft}
        if (InputFire := MaybeInputFire?):
            InputFire.Enable()
            spawn. InputTriggerListenerPressed(InputFire, FirePressedEvent)
            set AllInputDevices += array{InputFire}
        if (BoundAgent := MaybeBoundAgent?) then RegisterToAllDevices(BoundAgent)


    OnEndSimulation<override>():void =
        (super:)OnEndSimulation()
        UnbindAgent()
        EndSimulationEvent.Signal()


    BindAgent<public>(Agent: agent): void =
        if (MaybeBoundAgent?) then UnbindAgent()
        set MaybeBoundAgent = option. Agent
        RegisterToAllDevices(Agent)
        AgentBoundEvent.Signal(Agent)


    UnbindAgent<public>(): void =
        if (RegisteredAgent := MaybeBoundAgent?):
            UnregisterFromAllDevices(RegisteredAgent)
            set MaybeBoundAgent = false
            AgentUnboundEvent.Signal(RegisteredAgent)


    InputTriggerListenerPressed(InputDevice: input_trigger_device, EventToSignal: event())<suspends>: void =
        race:
            EndSimulationEvent.Await()
            InputsInitializedEvent.Await()
            loop:
                Agent := InputDevice.PressedEvent.Await()
                if (Agent = MaybeBoundAgent?):
                    EventToSignal.Signal()
                Sleep(0.0)


    OnAgentRemoved(Agent: agent): void =
        if (Agent = MaybeBoundAgent?):
            UnbindAgent()


    RegisterToAllDevices<internal>(Agent: agent):void =
        for (InputDevice : AllInputDevices): 
            InputDevice.Register(Agent)


    UnregisterFromAllDevices<internal>(Agent: agent):void =
        for (InputDevice : AllInputDevices): 
            InputDevice.Unregister(Agent)
    


